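# Sound cue for the close step of the Nighthawk six-shooter revolver's reload in the Figment pistol content.
# A quality-level node sends the first level to an equally weighted random pick between two close-sound waves, while the other three levels share a single wave player.
# It plays as weapon foley with shared reload attenuation, reduced volume and a concurrency override that prevents new instances once the limit is reached.
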SoundCue /Figment_Pistol_SixShooter_Athena/Sounds/FNS01_Pistol_Revolver_Nighthawk_Reload_Close_Cue
root=QualityLevel_0
nodes [
 {id=QualityLevel_0, children [Random_1,WavePlayer_4,WavePlayer_4,WavePlayer_4]}
 {id=Random_1, children [WavePlayer_2,WavePlayer_3], properties {weights [1,1]}}
 {id=WavePlayer_2, properties {soundWaveAssetPtr=&/revolver_reload_B_close_01}}
 {id=WavePlayer_3, properties {soundWaveAssetPtr=&/revolver_reload_B_close_02}}
 {id=WavePlayer_4, properties {soundWaveAssetPtr=&revolver_reload_B_close_01}}
]
properties {
 attenuationSettings=&Fort_Weapon_Reload_Att
 bOverrideConcurrency=true
 maxDistance=2250
 soundClassObject=&Fort_WeaponFoley
 volumeMultiplier=0.7
 concurrencyOverrides {resolutionRule=PreventNew}
}
