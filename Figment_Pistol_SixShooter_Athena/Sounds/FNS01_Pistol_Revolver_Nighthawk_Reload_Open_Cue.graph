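# Sound cue for the opening stage of the Nighthawk six-shooter revolver reload, classed as weapon foley.
# A mixer layers a modulated cylinder-open sample with a branch delayed by a fixed fraction of a second that randomly picks one of two reload samples.
# It uses the shared weapon reload attenuation and a concurrency override that prevents new instances rather than cutting playing ones.
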SoundCue /Figment_Pistol_SixShooter_Athena/Sounds/FNS01_Pistol_Revolver_Nighthawk_Reload_Open_Cue
root=Mixer_0
nodes [
 {id=Delay_0, children [Random_5], properties {delayMax=0.4, delayMin=0.4}}
 {id=Mixer_0, children [Modulator_1,Delay_0], properties {inputVolume [1,1]}}
 {id=Modulator_1, children [WavePlayer_6]}
 {id=Random_5, children [WavePlayer_13, WavePlayer_14], properties {weights [1,1]}}
 {id=WavePlayer_13, properties {soundWaveAssetPtr=&/revolver_reload_B_reaload_01}}
 {id=WavePlayer_14, properties {soundWaveAssetPtr=&/revolver_reload_B_reaload_02}}
 {id=WavePlayer_6, properties {soundWaveAssetPtr=&/revolver_reload_B_open_01}}
]
properties {
 attenuationSettings=&Fort_Weapon_Reload_Att
 bOverrideConcurrency=true
 maxDistance=2250
 soundClassObject=&Fort_WeaponFoley
 volumeMultiplier=0.7
 concurrencyOverrides {resolutionRule=PreventNew}
}
