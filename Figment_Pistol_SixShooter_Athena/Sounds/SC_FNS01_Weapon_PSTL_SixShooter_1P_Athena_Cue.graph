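# First-person Sound Cue for the Figment SixShooter pistol's gunshot, built from repurposed shotgun samples.
# A Mixer layers a boosted sharpshot wave with a quieter stereo shotgun fire tail that starts after a short fixed delay.
# Separate Modulators randomize each layer's pitch upward, raising the shot more than the tail.
# It uses first-person weapon attenuation, sound class and gunfire concurrency assets, and has its debug flag on.
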SoundCue /Figment_Pistol_SixShooter_Athena/Sounds/SC_FNS01_Weapon_PSTL_SixShooter_1P_Athena_Cue
root=Mixer_0
nodes [
 {id=Delay_4, children [Modulator_11], properties {delayMax=0.3, delayMin=0.3}}
 {id=Mixer_0, children [Modulator_10,Delay_4], properties {inputVolume [2,0.5]}}
 {id=Modulator_10, children [WavePlayer_2], properties {pitchMax=1.5, pitchMin=1.3}}
 {id=Modulator_11, children [WavePlayer_5], properties {pitchMax=1.3, pitchMin=1.1}}
 {id=WavePlayer_2, properties {soundWaveAssetPtr=&/shotgun_sharpshot_01}}
 {id=WavePlayer_5, properties {soundWaveAssetPtr=&/Shotgun_Fire_Tail_Stereo_01}}
]
properties {
 attenuationSettings=&Fort_Weapon_Projectile_FirstPerson_Athena_Att
 bDebug=true
 maxDistance=2097152
 soundClassObject=&Fort_WeaponRanged_Fire_1P_PSM
 volumeMultiplier=0.6
 attenuationOverrides {manualReverbSendLevel=-1.0165, reverbSendMethod=Manual}
 concurrencyOverrides {bLimitToOwner=true, maxCount=3, resolutionRule=StopOldest, volumeScale=0.9}
 concurrencySet [&Weapon_Gunfire_FirstPerson_Conc]
}
